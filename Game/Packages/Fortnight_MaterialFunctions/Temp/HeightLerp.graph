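# HeightLerp is a material function that blends two colors or textures using a grayscale heightmap and a transition phase.
# It offsets height minus one by twice the phase and clamps it, so B emerges through higher areas first as the phase rises.
# CheapContrast sharpens that mask by Falloff Contrast, despite the input description citing a power node.
# Outputs are the lerp result plus the alpha with and without contrast.
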MaterialFunction /Game/Packages/Fortnight_MaterialFunctions/Temp/HeightLerp
graph {
 properties {description=Lerp between two values or textures based off of a height map and a transition value., libraryCategoriesText [Misc,Texturing]}
 nodes [
  n1 = FunctionInput() {description=Enter a grayscale heightmap with values between 0-1., inputName=Height Texture, inputType=FunctionInput_Scalar, sortPriority=3}
  n2 = Constant() {r=1}
  n3 = FunctionInput() {
   bUsePreviewValueAsDefault=true
   description=Enter a value between 0-1.
   inputName=Transition Phase
   inputType=FunctionInput_Scalar
   sortPriority=3
  }
  n4 = Constant() {r=2}
  n5 = Clamp(None=Add(A=Subtract(A=@n1, B=@n2), B=Multiply(A=@n3, B=@n4)))
  FunctionOutput(None=ComponentMask(None=@n5) {r=true}) {description="Value used in lerp function. ", outputName=Lerp Alpha No Contrast, sortPriority=1}
  n6 = FunctionInput() {description=Enter a color or texture. This texture Acts as the base., inputName=Lerp A, previewValue=<1 1 1 1>}
  n7 = FunctionInput() {description=Enter a color or texture. This texture is exposed as the Transistion Phase increases., inputName=Lerp B, sortPriority=1}
  n8 = FunctionInput() {
   bUsePreviewValueAsDefault=true
   description="Uses power node (more expensive)"
   inputName=Falloff Contrast
   inputType=FunctionInput_Scalar
   sortPriority=3
   previewValue=<1 0 0 0>
  }
  n9 = MaterialFunctionCall(In=@n5, Contrast=@n8) {materialFunction=&CheapContrast}
  FunctionOutput(None=LinearInterpolate(A=@n6, B=@n7, Alpha=@n9.Result)) {description=Lerped result of value a and b based off of the input heightmap and transition value., outputName=Lerp Results}
  FunctionOutput(None=ComponentMask(None=@n9.Result) {r=true}) {description="Value used in lerp function. ", outputName=Lerp Alpha, sortPriority=1}
 ]
}
